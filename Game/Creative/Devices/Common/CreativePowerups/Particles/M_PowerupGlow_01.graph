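# Additive, unlit particle sprite material for a glow on Creative powerup devices.
# A mask texture drives a time-based surge modulated by panning noise, then tinted by particle color for emissive.
# A noise-distorted radial SphereMask shapes the color blend and opacity, with distance and depth fades that the mobile path skips.
# The graph also holds a sine height-bob lerp and a team-color palette lookup indexed by particle red.
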Material /Game/Creative/Devices/Common/CreativePowerups/Particles/M_PowerupGlow_01
graph {
 properties {
  bEnableResponsiveAA=true
  bUsedWithParticleSprites=true
  blendMode=BLEND_Additive
  pixelDepthOffsetMode=PDOM_Legacy
  shadingModel=MSM_Unlit
  translucencyPass=MTP_BeforeDOF
  usedShadingModels=Unlit
 }
 nodes [
  n1 = ParticleColor()
  Ceil(None=@n1.R)
  Constant()
  n2 = Time()
  n3 = Constant() {r=0.4}
  LinearInterpolate(
   A=VectorParameter() {parameterName=BobHeightMax, defaultValue=<rgba 0 0 5 0>}.RGB
   B=VectorParameter() {parameterName=BobHeightMin, defaultValue=<rgba 0 0 -1 0>}.RGB
   Alpha=Sine(None=Multiply(A=@n2, B=@n3))
  )
  n4 = TextureSample(UVs=Panner()) {samplerType=SAMPLERTYPE_LinearColor, texture=/Game/Creative/Devices/Common/CreativePowerups/Textures/T_PowerUp_Mask01_M}
  n5 = Add(A=Multiply(A=Time(), B=ScalarParameter() {defaultValue=0.3, parameterName=Surge Speed}) {constB=0.1}, B=@n4.A)
  n6 = Multiply(A=Saturate(None=Frac(None=@n5))) {constB=3}
  n7 = Saturate(None=Add(A=Subtract(A=@n6) {constB=2}, B=@n4.B))
  n8 = Multiply(A=Saturate(None=Sine(None=@n7)), B=Multiply(A=@n4.R, B=@n4.A))
  n9 = TextureSample(UVs=Panner() {speedX=0.15, speedY=-0.038}) {texture=&T_NoisePack_01}
  n10 = Multiply(A=Add(A=Multiply(A=@n8) {constB=5}, B=@n4.R), B=Multiply(A=@n9.R, B=ScalarParameter() {defaultValue=2, parameterName=Circuit Strength}))
  n11 = Multiply(A=@n1.RGB) {constB=0.5}
  n12 = Add(A=@n1.RGB) {constB=0.1}
  n13 = AppendVector(A=@n9.R, B=TextureSample(UVs=Panner() {speedX=-0.077, speedY=0.371}) {texture=&T_NoisePack_01}.G)
  n14 = Add(
   A=Multiply(A=ScalarParameter() {defaultValue=0.02, parameterName=NoiseScale}, B=ConstantBiasScale(None=@n13) {bias=-0.5, scale=2})
   B=TextureCoordinate()
  )
  n15 = Constant() {r=0.49}
  n16 = Constant() {r=0.5}
  n17 = Constant()
  n18 = SphereMask(
   A=Power(Base=SphereMask(A=@n14, B=@n15, Radius=@n16, Hardness=@n17), Exp=ScalarParameter() {defaultValue=1.75, parameterName=Exp})
   B=ScalarParameter() {defaultValue=0.2, parameterName=Center}
   Radius=ScalarParameter() {defaultValue=0.15, parameterName=Radius}
   Hardness=ScalarParameter() {parameterName=Hardness}
  )
  n19 = Power(Base=@n18, Exp=ScalarParameter() {defaultValue=4, parameterName=ColorLerpExponent})
  MaterialFunctionCall(Emissive=Multiply(A=Add(A=@n10), B=LinearInterpolate(A=@n11, B=@n12, Alpha=@n19))) {materialFunction=&MF_FX_EmissiveAdjustments}
  n20 = Power(Base=@n18)
  n21 = ScalarParameter() {parameterName="Fade_Start_Distance (S)"}
  n22 = Add(A=ParticleRadius(), B=ScalarParameter() {defaultValue=10, parameterName=RadiusBias})
  n23 = DepthFade(Opacity=MaterialFunctionCall(Opacity In=@n20, Position=ParticlePositionWS(), Fade_Start_Distance=Add(A=@n21, B=@n22), Fade_End_Distance=@n22) {materialFunction=&MF_Distance_SyncedFade_FX}.Result) {fadeDistanceDefault=10}
  ShadingPathSwitch(Default=@n23, Deferred=@n23, Mobile=@n20) {desc=Migrated from FeatureLevelSwitch}
  n24 = Multiply(B=@n1.R) {constA=4, desc=4PixelsForEachColor}
  TextureSampleParameter2D!(UVs=Divide(A=Add(A=@n24) {constB=2, desc=OffsetToCenterOfColor}) {constB=256, desc=TextureRes}, Level, Apply View MipBias) {
   constMipValue=0
   group=TeamColor
   mipValueMode=TMVM_MipLevel
   parameterName=TeamColors 1-64
   sortPriority=1
   texture=/Game/Creative/Materials/T_CreativeColors
  }
 ]
}
