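# Unlit translucent sprite-particle material for a soft, animated glow on Creative powerups.
# Two panning noise samples distort the UVs of a sphere mask, which is raised to a power and multiplied by a radial mask texture.
# Particle color tints that mask for boosted emissive through MF_FX_EmissiveAdjustments, and particle alpha scales it.
# The graph also holds a sine bob-height blend and a team-color palette lookup indexed by particle red.
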Material /Game/Creative/Devices/Common/CreativePowerups/Particles/M_PowerupGlow_02
graph {
 properties {
  bUsedWithParticleSprites=true
  blendMode=BLEND_Translucent
  pixelDepthOffsetMode=PDOM_Legacy
  shadingModel=MSM_Unlit
  usedShadingModels=Unlit
 }
 nodes [
  Constant()
  n1 = Time()
  n2 = Constant() {r=0.4}
  LinearInterpolate(
   A=VectorParameter() {parameterName=BobHeightMax, defaultValue=<rgba 0 0 5 0>}.RGB
   B=VectorParameter() {parameterName=BobHeightMin, defaultValue=<rgba 0 0 -1 0>}.RGB
   Alpha=Sine(None=Multiply(A=@n1, B=@n2))
  )
  n3 = ParticleColor()
  n4 = Add(A=@n3.RGB) {constB=0.1}
  n5 = AppendVector(
   A=TextureSample(UVs=Panner() {speedX=0.15, speedY=-0.038}) {texture=&T_NoisePack_01}.R
   B=TextureSample(UVs=Panner() {speedX=-0.077, speedY=0.371}) {texture=&T_NoisePack_01}.G
  )
  n6 = Add(
   A=Multiply(A=ScalarParameter() {defaultValue=0.02, parameterName=NoiseScale}, B=ConstantBiasScale(None=@n5) {bias=-0.5, scale=2})
   B=TextureCoordinate()
  )
  n7 = Constant() {r=0.49}
  n8 = Constant() {r=0.5}
  n9 = Constant()
  n10 = Multiply(
   A=Power(Base=SphereMask(A=@n6, B=@n7, Radius=@n8, Hardness=@n9), Exp=ScalarParameter() {defaultValue=1.75, parameterName=Exp})
   B=TextureSample() {texture=/Game/Effects/Fort_Effects/Textures/Mask/T_RadialMask}.G
  )
  MaterialFunctionCall(Emissive=Multiply(A=Multiply(A=@n4, B=@n10)) {constB=10}) {materialFunction=&MF_FX_EmissiveAdjustments}
  Multiply(A=@n3.A, B=@n10)
  n11 = Multiply(B=@n3.R) {constA=4, desc=4PixelsForEachColor}
  TextureSampleParameter2D!(UVs=Divide(A=Add(A=@n11) {constB=2, desc=OffsetToCenterOfColor}) {constB=256, desc=TextureRes}, Level, Apply View MipBias) {
   constMipValue=0
   group=TeamColor
   mipValueMode=TMVM_MipLevel
   parameterName=TeamColors 1-64
   sortPriority=1
   texture=/Game/Creative/Materials/T_CreativeColors
  }
 ]
}
