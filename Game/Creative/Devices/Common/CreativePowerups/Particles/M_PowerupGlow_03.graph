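# Unlit translucent particle sprite material for a glow on Creative device powerups.
# A spiky flare texture, scaled about its UV center, multiplies brightened particle color and is boosted through a shared emissive adjustment function.
# The same flare mask multiplied by particle alpha gives a soft transparency term.
# A time-driven sine blends between max and min height vector parameters to produce a vertical bobbing offset.
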Material /Game/Creative/Devices/Common/CreativePowerups/Particles/M_PowerupGlow_03
graph {
 properties {
  bUsedWithParticleSprites=true
  blendMode=BLEND_Translucent
  pixelDepthOffsetMode=PDOM_Legacy
  shadingModel=MSM_Unlit
  usedShadingModels=Unlit
 }
 nodes [
  n1 = Time()
  n2 = Constant() {r=0.4}
  LinearInterpolate(
   A=VectorParameter() {parameterName=BobHeightMax, defaultValue=<rgba 0 0 5 0>}.RGB
   B=VectorParameter() {parameterName=BobHeightMin, defaultValue=<rgba 0 0 -1 0>}.RGB
   Alpha=Sine(None=Multiply(A=@n1, B=@n2))
  )
  n3 = ParticleColor()
  n4 = Add(A=@n3.RGB) {constB=0.2}
  n5 = TextureSample(UVs=MaterialFunctionCall(Texture Scale=ScalarParameter() {defaultValue=2.1, parameterName=TextureScale}) {materialFunction=&ScaleUVsByCenter}.UVs) {texture=/Game/Effects/Fort_Effects/Textures/Flares/T_Flare_Spikeball}
  MaterialFunctionCall(Emissive=Multiply(A=Multiply(A=@n4, B=@n5.R)) {constB=10}) {materialFunction=&MF_FX_EmissiveAdjustments}
  Multiply(A=@n3.A, B=@n5.R)
 ]
}
